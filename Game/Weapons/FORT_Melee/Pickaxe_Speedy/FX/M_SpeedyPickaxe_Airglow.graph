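# Unlit additive mesh-particle material for the Speedy pickaxe's airglow effect, producing a Fresnel-based glow.
# Fresnel times particle alpha blends two colors, which particle color then tints before an emissive adjustment function.
# A static switch can scale brightness down at night through DayNightLerp.
# The same Fresnel mask forms an opacity-style output, with optional depth fade on the default shading path and none on mobile.
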Material /Game/Weapons/FORT_Melee/Pickaxe_Speedy/FX/M_SpeedyPickaxe_Airglow
graph {
 properties {
  bUsedWithMeshParticles=true
  bUsedWithNiagaraMeshParticles=true
  blendMode=BLEND_Additive
  pixelDepthOffsetMode=PDOM_Legacy
  shadingModel=MSM_Unlit
  usedShadingModels=Unlit
 }
 nodes [
  n1 = ParticleColor()
  n2 = Multiply(
   A=@n1.A
   B=MaterialFunctionCall(
    Brightness=ScalarParameter() {defaultValue=1, parameterName=Fresnel Brightness}
    Exponent=ScalarParameter() {defaultValue=10, parameterName=Fresnel Exponent}
   ) {materialFunction=&MF_Fresnel_FX}.Fresnel
  )
  n3 = Multiply(
   A=LinearInterpolate(
    A=VectorParameter() {parameterName=Color_A, defaultValue=<rgba 0.0624 0.7188 0.3641 1>}.RGB
    B=VectorParameter() {parameterName=Color_B, defaultValue=<rgba 0.5515 2.3842 4 1>}.RGB
    Alpha=@n2
   )
   B=@n1.RGB
  )
  n4 = Multiply(A=@n3, B=ScalarParameter() {defaultValue=0.4, parameterName=Night Multiplier})
  n5 = MaterialFunctionCall() {materialFunction=/Game/Weapons/WeaponSkins/MaterialFunction/DayNightLerp}
  MaterialFunctionCall(Emissive=StaticSwitchParameter(True=LinearInterpolate(A=@n4, B=@n3, Alpha=@n5.Result), False=@n3) {parameterName=Use Day Night Multiplier}) {materialFunction=&MF_FX_EmissiveAdjustments}
  n6 = ScalarParameter() {defaultValue=16, parameterName=Depth Fade Distance}
  ShadingPathSwitch(Default=StaticSwitchParameter(True=DepthFade(Opacity=@n2, FadeDistance=@n6), False=@n2) {parameterName=Use DepthFade}, Mobile=@n2) {desc=Migrated from FeatureLevelSwitch}
 ]
}
